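# Additive, unlit, two-sided beam-trail material for the airfoil trails of the female Black Monday glider.
# A panning claw-mark texture, on swizzled UVs offset by the ScracthMove parameter, brightens the particle color before the emissive adjustment function.
# Two smoke textures pan at different rates on the second UV channel and multiply into noise with a raised floor.
# That noise, an arc mask, particle alpha and an inverted V gradient multiply into the trail's fade.
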Material /BRCosmetics/Gadgets/Assets/VinderTech_GliderChute/Glider_FemaleBlackMonday/FX/M_Glider_BlackMondayFem_AirFoilTrails
graph {
 properties {
  bUsedWithBeamTrails=true
  blendMode=BLEND_Additive
  pixelDepthOffsetMode=PDOM_Legacy
  shadingModel=MSM_Unlit
  twoSided=true
  usedShadingModels=Unlit
 }
 nodes [
  n1 = ParticleColor()
  n2 = MaterialFunctionCall(XY=TextureCoordinate()) {materialFunction=&Swizzle}
  n3 = ComponentMask(None=VectorParameter() {parameterName=ScracthMove, defaultValue=<rgba 0.2 0 0 0>}.RGB) {g=true, r=true}
  n4 = TextureSample(UVs=Panner(Coordinate=Add(A=@n2.YX, B=@n3)) {speedY=-3}) {texture=&/T_Clawmark}
  MaterialFunctionCall(Emissive=LinearInterpolate(A=@n1.RGB, B=Multiply(A=@n1.RGB) {constB=1.5}, Alpha=Add(A=@n4.R, B=@n4.R))) {materialFunction=&MF_FX_EmissiveAdjustments}
  n5 = Panner(Coordinate=TextureCoordinate() {coordinateIndex=1, uTiling=0.1}) {speedX=-0.1874}
  n6 = Panner(Coordinate=TextureCoordinate() {coordinateIndex=1, uTiling=0.25}) {speedX=-0.25}
  n7 = LinearInterpolate(Alpha=Multiply(
    A=TextureSample(UVs=@n5) {texture=/Game/Effects/Fort_Effects/Textures/Smoke/T_Puffy_Smoke_Trail_01}.R
    B=TextureSample(UVs=@n6) {samplerType=SAMPLERTYPE_Grayscale, texture=&T_Smoke_Trail_Soft}.R
   )) {constA=0.1}
  n8 = TextureSample(UVs=TextureCoordinate()) {texture=/Game/Effects/Fort_Effects/Textures/Melee/T_Trail_ArcMask}
  n9 = TextureCoordinate()
  Multiply(A=Multiply(A=@n1.A, B=Multiply(A=@n7, B=@n8.R)), B=OneMinus(None=ComponentMask(None=@n9) {g=true}))
 ]
}
