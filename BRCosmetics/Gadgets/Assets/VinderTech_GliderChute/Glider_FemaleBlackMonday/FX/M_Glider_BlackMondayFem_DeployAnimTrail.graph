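# Translucent, unlit, two-sided trail material for the female Black Monday glider's deploy animation, usable on beam trails and Niagara sprites.
# Particle color RGB passes through a shared emissive adjustment function.
# A claw-mark texture with swizzled UVs has its red channel doubled and multiplied by particle alpha as an opacity mask.
# A shading path switch adds parameter-blended depth fade on the default path, while mobile uses the plain mask.
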Material /BRCosmetics/Gadgets/Assets/VinderTech_GliderChute/Glider_FemaleBlackMonday/FX/M_Glider_BlackMondayFem_DeployAnimTrail
graph {
 properties {
  bUsedWithBeamTrails=true
  bUsedWithNiagaraSprites=true
  blendMode=BLEND_Translucent
  pixelDepthOffsetMode=PDOM_Legacy
  shadingModel=MSM_Unlit
  twoSided=true
  usedShadingModels=Unlit
 }
 nodes [
  n1 = ParticleColor()
  MaterialFunctionCall(Emissive=@n1.RGB) {materialFunction=&MF_FX_EmissiveAdjustments}
  n2 = TextureSample(UVs=MaterialFunctionCall(XY=TextureCoordinate()) {materialFunction=&Swizzle}.YX) {texture=&/T_Clawmark}
  n3 = Multiply(A=@n1.A, B=Add(A=@n2.R, B=@n2.R))
  n4 = ScalarParameter() {parameterName=FadeDistance}
  ShadingPathSwitch(
   Default=LinearInterpolate(A=@n3, B=DepthFade(Opacity=@n3, FadeDistance=@n4), Alpha=ScalarParameter() {parameterName=UseDepthFade}) {constA=1}
   Mobile=@n3
  ) {desc=Migrated from FeatureLevelSwitch}
 ]
}
